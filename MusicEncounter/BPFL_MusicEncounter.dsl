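; Blueprint function library that broadcasts music encounter gameplay events through a supplied Global Event Router component.
; Four functions identify the event by enum byte, name, instanced struct or gameplay tag, wrapping it in the matching MusicEncounter event struct.
; Each broadcasts with a context object only when cosmetic events can execute and the router reference is valid.
; The tag variant also branches on the broadcast result, with no action attached.
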(fn ExecuteMusicEncounterGameplayEventbyEnum
  (params (Global%20Event%20Router :type Gameplay%20Event%20Router%20Component%20Object%20Reference) (Event%20Enum :type Byte) (Context :type Object%20Reference))
  (Networking|CanExecuteCosmeticEvents
    (:True
      (bind _global_event_router (Variables|GetGlobal%20Event%20Router)
        (:then
          (GameplayEventRouter|BroadcastEvent _global_event_router (Utilities|Struct|MakeGameplayEventMusicEncounterEnum Event%20Enum) (Variables|GetContext)))
        (:else)))))

(fn ExecuteMusicEncounterGameplayEventbyName
  (params (Global%20Event%20Router :type Gameplay%20Event%20Router%20Component%20Object%20Reference) (Event%20Name :type Name) (Context :type Object%20Reference))
  (Networking|CanExecuteCosmeticEvents
    (:True
      (bind _global_event_router (Variables|GetGlobal%20Event%20Router)
        (:then
          (GameplayEventRouter|BroadcastEvent _global_event_router (Utilities|Struct|MakeGameplayEventMusicEncounterName Event%20Name) (Variables|GetContext)))
        (:else)))))

(fn ExecuteMusicEncounterGameplayEventbyStruct
  (params (Global%20Event%20Router :type Gameplay%20Event%20Router%20Component%20Object%20Reference) (Instanced%20Struct :type Instanced%20Struct%20Structure) (Context :type Object%20Reference))
  (Networking|CanExecuteCosmeticEvents
    (:True
      (bind _global_event_router (Variables|GetGlobal%20Event%20Router)
        (:then
          (GameplayEventRouter|BroadcastEvent _global_event_router (Utilities|Struct|MakeGameplayEventMusicEncounterStruct Instanced%20Struct) (Variables|GetContext)))
        (:else)))))

(fn ExecuteMusicEncounterGameplayEventbyTag
  (params (Global%20Event%20Router :type Gameplay%20Event%20Router%20Component%20Object%20Reference) (Gameplay%20Tag :type Gameplay%20Tag%20Structure) (Context :type Object%20Reference))
  (Networking|CanExecuteCosmeticEvents
    (:True
      (bind _global_event_router (Variables|GetGlobal%20Event%20Router)
        (:then
          (bind _returnvalue (GameplayEventRouter|BroadcastEvent _global_event_router (Utilities|Struct|MakeGameplayEventMusicEncounter Gameplay%20Tag) (Variables|GetContext)))
          (if _returnvalue
            _))
        (:else)))))
