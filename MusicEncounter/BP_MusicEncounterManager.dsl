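; Manager that creates, tracks and shuts down music encounter instances.
; RequestMusicEncounter reuses an existing instance matching the gameplay tag, or assigns a new GUID and queues it under its soft class, so concurrent requests share one async load.
; Once loaded, the queued instances are constructed, mapped by ID and announced through ready events.
; It also lazily creates a music registrar, spawns audio gameplay volumes, and uninitializes encounters on EndPlay or StopAllMusicEncounters.
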(event Custom|Load%20Music%20Encounter%20Class
  (params (Music%20Encounter%20Class :type Object%20Soft%20Class%20Reference))
  (bind _class (Utilities|AsyncLoadClassAsset Music%20Encounter%20Class)
    (:Completed
      (bind _asmusic_encounter_base (Utilities|Casting|CastToBP_MusicEncounterClass _class))
      (MusicEncounterManager|HandleMusicEncounterClassLoaded _asmusic_encounter_base))))

(event Custom|Create%20Music%20Encounter%20Instances
  (params (Music%20Encounter%20Class :type Music%20Encounter%20Base%20Class%20Reference) (Instance%20IDs :type Array%20of%20Guid%20Structures))
  (for _array_element Instance%20IDs
    (CallFunction|CreateMusicEncounterInstance Music%20Encounter%20Class _array_element)))

(event Custom|Create%20Music%20Encounter%20Instance
  (params (Music%20Encounter%20Class :type Music%20Encounter%20Base%20Class%20Reference) (Instance%20ID :type Guid%20Structure))
  (bind _returnvalue (Game|ConstructObjectfromClass Music%20Encounter%20Class))
  (MusicEncounter|SetInstanceID _returnvalue Instance%20ID)
  (Utilities|Map|Add (Variables|MusicEncounterManager|GetMusic%20Encounters) Instance%20ID _returnvalue)
  (MusicEncounter|SetInstanceTag _returnvalue (Utilities|Map|Find (Variables|MusicEncounterManager|GetMusic%20Encounter%20Instance%20Tags) Instance%20ID))
  (MusicEncounterManager|CallOnMusicEncounterConstructed _returnvalue)
  (MusicEncounterManager|CallOnMusicEncounterReady _returnvalue Instance%20ID))

(event EventEndPlay
  (params (EndPlayReason :type EEndPlayReason%20Enum :default "Destroyed"))
  (bind (_keypin _valuepin) (Utilities|Map|ForEachLoop(Map) (Variables|MusicEncounterManager|GetMusic%20Encounters)))
  (Class|MusicEncounterBase|UninitializeMusicEncounter _valuepin))

(fn GetorCreateMusicRegistrar
  (bind _music_registrar (Variables|MusicEncounterManager|GetMusic%20Registrar)
    (:then
      (return _music_registrar))
    (:else
      (bind _music_registrar_class (Variables|MusicEncounterManager|GetMusic%20Registrar%20Class)
        (:then
          (bind _returnvalue (Game|ConstructObjectfromClass _music_registrar_class))
          (bind _output_get (Variables|MusicEncounterManager|SetMusic%20Registrar _returnvalue))
          (return _output_get))
        (:else)))))

(fn HandleMusicEncounterClassLoaded
  (params (Music%20Encounter%20Class :type Music%20Encounter%20Base%20Class%20Reference))
  (bind _value (Utilities|Map|Find (Variables|MusicEncounterManager|GetPending%20Loads) (Utilities|ResolveSoftReference (Variables|GetMusic%20Encounter%20Class))))
  (bind _value_instanceids_3_003962c34a306b75d5fa708070703322 (Utilities|Struct|BreakMusicEncounterInstanceIDs _value))
  (Utilities|FlowControl|Sequence
    (:then_0
      (LocalVariables|SetInstance%20IDs _value_instanceids_3_003962c34a306b75d5fa708070703322))
    (:then_1
      (Utilities|FlowControl|Sequence
        (:then_0
          (Utilities|Map|Remove (Variables|MusicEncounterManager|GetPending%20Loads) (Utilities|ResolveSoftReference (Variables|GetMusic%20Encounter%20Class))))
        (:then_1
          (Utilities|FlowControl|Sequence
            (:then_0
              (CallFunction|CreateMusicEncounterInstances (Variables|GetMusic%20Encounter%20Class) (LocalVariables|GetInstance%20IDs)))))))))

(fn RequestMusicEncounter
  (params (Music%20Encounter%20Instance%20Tag :type Gameplay%20Tag%20Structure) (Music%20Encounter%20Class :type Music%20Encounter%20Base%20Soft%20Class%20Reference) (On%20Music%20Encounter%20Ready :type Delegate%20%28by%20ref%29))
  (bind _music_encounter_instance_tag (Variables|GetMusic%20Encounter%20Instance%20Tag))
  (bind _tagged_instance_ids (Variables|MusicEncounterManager|GetTagged%20Instance%20IDs))
  (bind _value (Utilities|Map|Find _tagged_instance_ids _music_encounter_instance_tag))
  (bind _music_encounters (Variables|MusicEncounterManager|GetMusic%20Encounters))
  (bind (_value_2 _returnvalue) (Utilities|Map|Find _music_encounters _value))
  (bind _music_encounter_class (Variables|GetMusic%20Encounter%20Class))
  (bind _pending_loads (Variables|MusicEncounterManager|GetPending%20Loads))
  (bind (_value_3 _returnvalue_2) (Utilities|Map|Find _pending_loads _music_encounter_class))
  (bind (_value_4 _returnvalue_3) (Utilities|Map|Find (Variables|MusicEncounterManager|GetTagged%20Instance%20IDs) (Variables|GetMusic%20Encounter%20Instance%20Tag)))
  (if (Variables|MusicEncounterManager|GetEnable%20Music%20Encounter%20System)
    (MusicEncounterManager|BindEventtoOnMusicEncounterReady On%20Music%20Encounter%20Ready)
    (if (GameplayTags|IsGameplayTagValid (Variables|GetMusic%20Encounter%20Instance%20Tag))
      ;[1 1 1] Check if the component's instance tag matches an encounter that already exists. If so, return the found instance.
      (if _returnvalue
        (MusicEncounterManager|CallOnMusicEncounterReady _value_2 _value)
        (elif _returnvalue_3
          _))
      (else
        _))
    (Variables|GetMusic%20Encounter%20Class
      (:then
        ;[1 1 1] We're creating a new instance.
        (LocalVariables|SetNew%20Instance%20ID (Guid|NewGuid))
        (Utilities|FlowControl|Sequence
          (:then_0
            ;[1 1 1] Map the instance tags and IDs for easy lookup
            (if (GameplayTags|IsGameplayTagValid (Variables|GetMusic%20Encounter%20Instance%20Tag))
              (Utilities|Map|Add (Variables|MusicEncounterManager|GetTagged%20Instance%20IDs) (Variables|GetMusic%20Encounter%20Instance%20Tag) (LocalVariables|GetNew%20Instance%20ID))
              (Utilities|Map|Add (Variables|MusicEncounterManager|GetMusic%20Encounter%20Instance%20Tags) (LocalVariables|GetNew%20Instance%20ID) (Variables|GetMusic%20Encounter%20Instance%20Tag))))
          (:then_1
            (if _returnvalue_2
              ;[1 1 1] If so, add another instance to be created upon finishing loading
              (bind _output_get (LocalVariables|SetInstance%20IDs (Utilities|Struct|BreakMusicEncounterInstanceIDs _value_3)))
              (Utilities|Array|Add _output_get (LocalVariables|GetNew%20Instance%20ID))
              (Utilities|Map|Add (Variables|MusicEncounterManager|GetPending%20Loads) (Variables|GetMusic%20Encounter%20Class) (Utilities|Struct|MakeMusicEncounterInstanceIDs _output_get))
              (else
                ;[1 1 1] Add the first instance to be created upon finishing loading
                (Utilities|Map|Add (Variables|MusicEncounterManager|GetPending%20Loads) (Variables|GetMusic%20Encounter%20Class) (Utilities|Struct|MakeMusicEncounterInstanceIDs (Utilities|Array|MakeArray (LocalVariables|GetNew%20Instance%20ID))))
                (CallFunction|LoadMusicEncounterClass (Variables|GetMusic%20Encounter%20Class))))
            (return (LocalVariables|GetNew%20Instance%20ID)))))
      (:else
        (Development|PrintText (LocalVariables|GetRequest%20Failed%20Log) false)))))

(fn SpawnAudioGameplayVolume
  (params (Music%20Encounter%20AGV%20Class :type Music%20Encounter%20AGV%20Class%20Reference) (Music%20Encounter :type Music%20Encounter%20Base%20Object%20Reference))
  (bind _music_encounter_agv_class (Variables|GetMusic%20Encounter%20AGV%20Class)
    (:then
      (bind _returnvalue (Game|SpawnActorfromClass _music_encounter_agv_class (struct :Location "0, 0, 0" :Rotation "0, 0, 0" :Scale "1.000000,1.000000,1.000000") "Undefined" "MultiplyWithRoot" 0 Music%20Encounter 0))
      (return _returnvalue))
    (:else)))

(fn StopAllMusicEncounters
  (bind _values (Utilities|Map|Values (Variables|MusicEncounterManager|GetMusic%20Encounters)))
  (for _array_element _values
    (Class|MusicEncounterBase|UninitializeMusicEncounter _array_element)))
